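# Control Rig for the SafeThymeMask face accessory on the medium male body, correcting the mask's fur as the head moves.
# Head rotation relative to spine_05 weights an offset on the deform_mask_back helper bone.
# Averaged parent constraints then blend the dyn_fur_mask bones between spine_05, clavicles, head and that helper.
# Finally, head delta roll, remapped separately for each direction, adds rotation and translation to deform_mask_back, after the constraints have run.
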ControlRigBlueprint /BRCosmetics/Characters/Player/Male/Medium/Bodies/M_MED_SafeThymeMask/Meshes/Parts/M_MED_SafeThymeMask_FaceAcc_CtrlRig
graphs [
 RigVMFunctionLibrary {links [], nodes []}
 RigVMModel {nodes [
   RigUnit_BeginExecution = &RigUnit_BeginExecution()
   &RigUnit_ModifyTransforms(
    ItemToModify=((Item=(Type=Bone,Name="head"),Transform=(Rotation=(X=0,Y=0,Z=0,W=1),Translation=(X=0,Y=0,Z=0),Scale3D=(X=1,Y=1,Z=1))))
    Weight=1
    WeightMinimum=0
    ExecutePin=@RigUnit_BeginExecution.ExecutePin
   )
   n1 = &RigUnit_GetRelativeTransformForItem(Child=(Type=Bone,Name="head"), Parent=(Type=Bone,Name="spine_05"), bChildInitial=false, bParentInitial=false)
   > &RigUnit_ModifyTransforms(
    ItemToModify=((Item=(Type=Bone,Name="deform_mask_back"),Transform=(Rotation=(X=0,Y=0.4226,Z=0,W=0.9063),Translation=(X=10,Y=0,Z=16),Scale3D=(X=1,Y=1,Z=1))))
    WeightMinimum=0
    Weight=&RigVMFunction_MathDoubleRemap(
     SourceMaximum=-90
     SourceMinimum=0
     TargetMaximum=1
     TargetMinimum=0
     Value=&RigVMFunction_MathQuaternionToEuler(RotationOrder=ZYX, Value=@n1.RelativeTransform.Rotation).Result.Z
    ).Result
   )
   > &RigUnit_ParentConstraint(
    AdvancedSettings=(InterpolationType=Average,RotationOrderForFilter=XZY)
    Child=(Type=Bone,Name="dyn_fur_mask_fr_l")
    Parents=((Item=(Type=Bone,Name="spine_05"),Weight=1),(Item=(Type=Bone,Name="clavicle_l"),Weight=0.5))
    Weight=1
   )
   > &RigUnit_ParentConstraint(
    AdvancedSettings=(InterpolationType=Average,RotationOrderForFilter=XZY)
    Child=(Type=Bone,Name="dyn_fur_mask_fr_mid")
    Parents=((Item=(Type=Bone,Name="spine_05"),Weight=1))
    Weight=1
   )
   > &RigUnit_ParentConstraint(
    AdvancedSettings=(InterpolationType=Average,RotationOrderForFilter=XZY)
    Child=(Type=Bone,Name="dyn_fur_mask_fr_r")
    Parents=((Item=(Type=Bone,Name="spine_05"),Weight=1),(Item=(Type=Bone,Name="clavicle_r"),Weight=0.5))
    Weight=1
   )
   > &RigUnit_ParentConstraint(
    AdvancedSettings=(InterpolationType=Average,RotationOrderForFilter=XZY)
    Child=(Type=Bone,Name="dyn_fur_mask_side_01_l")
    Parents=((Item=(Type=Bone,Name="spine_05"),Weight=1),(Item=(Type=Bone,Name="clavicle_l"),Weight=1))
    Weight=1
   )
   > &RigUnit_ParentConstraint(
    AdvancedSettings=(InterpolationType=Average,RotationOrderForFilter=XZY)
    Child=(Type=Bone,Name="dyn_fur_mask_side_01_r")
    Parents=((Item=(Type=Bone,Name="spine_05"),Weight=1),(Item=(Type=Bone,Name="clavicle_r"),Weight=1),(Item=(Type=Bone,Name="head"),Weight=1))
    Weight=1
    bMaintainOffset=true
   )
   > &RigUnit_ParentConstraint(
    AdvancedSettings=(InterpolationType=Average,RotationOrderForFilter=XZY)
    Child=(Type=Bone,Name="dyn_fur_mask_side_02_l")
    Parents=((Item=(Type=Bone,Name="spine_05"),Weight=1),(Item=(Type=Bone,Name="clavicle_l"),Weight=1))
    Weight=1
   )
   > &RigUnit_ParentConstraint(
    AdvancedSettings=(InterpolationType=Average,RotationOrderForFilter=XZY)
    Child=(Type=Bone,Name="dyn_fur_mask_bk_l")
    Parents=((Item=(Type=Bone,Name="spine_05"),Weight=1),(Item=(Type=Bone,Name="clavicle_l"),Weight=1),(Item=(Type=Bone,Name="deform_mask_back"),Weight=0.4))
    Weight=1
   )
   > &RigUnit_ParentConstraint(
    AdvancedSettings=(InterpolationType=Average,RotationOrderForFilter=XZY)
    Child=(Type=Bone,Name="dyn_fur_mask_bk_up")
    Parents=((Item=(Type=Bone,Name="spine_05"),Weight=0.5),(Item=(Type=Bone,Name="head"),Weight=1))
    Weight=1
   )
   > &RigUnit_ParentConstraint(
    AdvancedSettings=(InterpolationType=Average,RotationOrderForFilter=XZY)
    Child=(Type=Bone,Name="dyn_fur_mask_bk_dw")
    Parents=((Item=(Type=Bone,Name="spine_05"),Weight=1),(Item=(Type=Bone,Name="head"),Weight=1),(Item=(Type=Bone,Name="deform_mask_back"),Weight=1))
    Weight=1
   )
   > &RigUnit_ParentConstraint(
    AdvancedSettings=(InterpolationType=Average,RotationOrderForFilter=XZY)
    Child=(Type=Bone,Name="dyn_fur_mask_side_02_r")
    Parents=((Item=(Type=Bone,Name="spine_05"),Weight=1),(Item=(Type=Bone,Name="clavicle_r"),Weight=1),(Item=(Type=Bone,Name="head"),Weight=0))
    Weight=1
   )
   > &RigUnit_ParentConstraint(
    AdvancedSettings=(InterpolationType=Average,RotationOrderForFilter=XZY)
    Child=(Type=Bone,Name="dyn_fur_mask_bk_r")
    Parents=((Item=(Type=Bone,Name="spine_05"),Weight=1),(Item=(Type=Bone,Name="clavicle_r"),Weight=1),(Item=(Type=Bone,Name="deform_mask_back"),Weight=0.4))
    Weight=1
   )
   > &RigUnit_ModifyTransforms(
    ItemToModify=((Item=(Type=Bone,Name="head"),Transform=(Rotation=(X=0,Y=0,Z=0,W=1),Translation=(X=0,Y=0,Z=0),Scale3D=(X=1,Y=1,Z=1))))
    Weight=1
    WeightMinimum=0
   )
   > Get Delta Ref = RigVMFunctionReferenceNode(Item=(Type=Bone,Name="head"))
   Remap = &RigVMFunction_MathDoubleRemap(SourceMaximum=80, SourceMinimum=0, TargetMaximum=75, TargetMinimum=0, Value=@"Get Delta Ref".Result.Rotation.Roll)
   Multiply = &RigVMFunction_MathDoubleMul(B=-0.1, A=@Remap.Result)
   n2 = &RigVMFunction_MathDoubleRemap(SourceMaximum=-80, SourceMinimum=0, TargetMaximum=-75, TargetMinimum=0, Value=@"Get Delta Ref".Result.Rotation.Roll)
   Add = &RigVMFunction_MathDoubleAdd(A=@Remap.Result, B=@n2.Result)
   RigVMFunction_MathQuaternionFromEuler = &RigVMFunction_MathQuaternionFromEuler(Euler=(X=0,Y=0,Z=0), Euler.X=@Add.Result)
   > &RigUnit_ModifyTransforms(
    ItemToModify=((Item=(Type=Bone,Name="deform_mask_back"),Transform=(Rotation=(X=0.4617,Y=0,Z=0,W=0.887),Translation=(X=0,Y=-10,Z=0),Scale3D=(X=1,Y=1,Z=1))))
    Weight=1
    WeightMinimum=0
    ItemToModify.0.Transform.Translation.Y=&RigVMFunction_MathDoubleAdd(A=@Multiply.Result, B=&RigVMFunction_MathDoubleMul(B=-0.1, A=@n2.Result).Result).Result
    ItemToModify.0.Transform.Translation.Z=&RigVMFunction_MathDoubleAdd(
     A=&RigVMFunction_MathDoubleMul(B=0.2, A=@Remap.Result).Result
     B=&RigVMFunction_MathDoubleMul(B=-0.2, A=@n2.Result).Result
    ).Result
    ItemToModify.0.Transform.Rotation=@RigVMFunction_MathQuaternionFromEuler.Result
   )
  ]}
]
hierarchy [
 root {loc=<0 0 0>, rot=<0 0 0 1>}
 pelvis {parent=root, loc=<0 -0.7071 93.1191>, rot=<0.0127 0.707 -0.0127 -0.707>}
 spine_01 {parent=pelvis, loc=<6.7494 -0.6388 0>, rot=<0 0 0.034 -0.9994>}
 spine_02 {parent=spine_01, loc=<7.2518 -0.035 0>, rot=<0 0 0.0238 -0.9997>}
 spine_03 {parent=spine_02, loc=<6.0104 0 0>, rot=<0 0 0.0149 -0.9999>}
 spine_04 {parent=spine_03, loc=<8.1024 0 0>, rot=<0 0 -0.0132 -0.9999>}
 spine_05 {parent=spine_04, loc=<10.0146 0 0>, rot=<0 0 -0.0015 -1>}
 neck_01 {parent=spine_05, loc=<11.3138 0.1529 0>, rot=<0 0 -0.2003 -0.9797>}
 neck_02 {parent=neck_01, loc=<4.5627 0 0>, rot=<0 0 0.0504 -0.9987>}
 head {parent=neck_02, loc=<5.3914 0 0>, rot=<0 0 0.115 -0.9934>}
 dyn_ear_base_r {parent=head, loc=<15.5703 -2.5361 9.5534>, rot=<0.9164 -0.0719 0.3933 0.0202>}
 dyn_ear_01_r {parent=dyn_ear_base_r, loc=<0 0 0>, rot=<0 0 0 1>}
 dyn_ear_base_l {parent=head, loc=<15.5702 -2.5361 -9.5534>, rot=<-0.0719 -0.9164 -0.0202 0.3933>}
 dyn_ear_01_l {parent=dyn_ear_base_l, loc=<0 0 0>, rot=<0 0 0 1>}
 dyn_fur_mask_bk_dw {parent=head, loc=<-7.3278 -13.4394 0>, rot=<0.0029 -0.7071 0.7071 -0.0029>}
 dyn_fur_mask_bk_r {parent=head, loc=<-3.646 -11.5318 9.4271>, rot=<0.0021 -0.5108 0.8597 -0.0036>}
 dyn_fur_mask_side_01_r {parent=head, loc=<-6.8441 2.8616 13.0941>, rot=<-0.0325 0.3484 0.9137 -0.2064>}
 dyn_fur_mask_fr_r {parent=head, loc=<-10.1385 7.485 6.3924>, rot=<-0.0809 0.5432 0.814 -0.1893>}
 dyn_fur_mask_side_02_r {parent=head, loc=<-2.9684 -4.8289 14.1085>, rot=<-0.0016 -0.1508 0.9885 -0.0108>}
 dyn_fur_mask_bk_up {parent=head, loc=<0.194 -13.3772 0>, rot=<0.0029 -0.7071 0.7071 -0.0029>}
 dyn_fur_mask_fr_l {parent=head, loc=<-10.1384 7.485 -6.3924>, rot=<-0.5432 -0.0809 -0.1893 -0.814>}
 dyn_fur_mask_side_01_l {parent=head, loc=<-6.8441 2.8616 -13.0941>, rot=<-0.3484 -0.0325 -0.2064 -0.9137>}
 dyn_fur_mask_side_02_l {parent=head, loc=<-2.9684 -4.829 -14.1085>, rot=<0.1508 -0.0016 -0.0108 -0.9885>}
 dyn_fur_mask_bk_l {parent=head, loc=<-3.646 -11.5318 -9.4271>, rot=<0.5108 0.0021 -0.0036 -0.8597>}
 dyn_fur_mask_fr_mid {parent=head, loc=<-11.4402 8.2985 0>, rot=<-0.0029 0.7071 0.7071 -0.0029>}
 clavicle_l {parent=spine_05, loc=<4.3575 -0.1466 -2.3223>, rot=<0.0465 -0.7289 -0.009 -0.6829>}
 clavicle_r {parent=spine_05, loc=<4.3575 -0.1466 2.3223>, rot=<0.7289 0.0465 -0.6829 0.009>}
 deform_mask_back {parent=spine_05, loc=<17.0927 -3.904 0>, rot=<0.0284 -0.7065 0.7065 -0.0284>}
]
